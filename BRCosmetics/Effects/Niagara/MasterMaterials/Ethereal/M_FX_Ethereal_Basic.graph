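# Translucent, lit master material for ethereal cosmetic effects on Niagara particles and skeletal meshes.
# Fresnel blends core and edge colors, from parameters or particle color plus a dynamic parameter.
# A scrolling, optionally screen-offset pattern, distorted by normals derived from the same texture, modulates that color before emissive adjustment.
# The blended alpha gets optional depth fade and a flippable smoothstep fade along UV V.
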Material /BRCosmetics/Effects/Niagara/MasterMaterials/Ethereal/M_FX_Ethereal_Basic
graph {
 properties {
  bAutomaticallySetUsageInEditor=false
  bUsedWithMeshDeformer=true
  bUsedWithNanite=true
  bUsedWithNiagaraMeshParticles=true
  bUsedWithNiagaraRibbons=true
  bUsedWithNiagaraSprites=true
  bUsedWithSkeletalMesh=true
  blendMode=BLEND_Translucent
  usedShadingModels=Default Lit
 }
 nodes [
  n1 = Fresnel(
   ExponentIn=ScalarParameter() {defaultValue=5, group=Frenel, parameterName=Edge Thickness}
   BaseReflectFractionIn=ScalarParameter() {defaultValue=0.05, group=Frenel, parameterName=Edge Softness}
  ) {baseReflectFraction=0.2}
  n2 = VectorParameter() {group=Color, parameterName=Core Color, defaultValue=<rgba 0.0638 0.1598 0.349 0.5>}
  n3 = VectorParameter() {group=Color, parameterName=Edge Color, sortPriority=33, defaultValue=<rgba 0.0677 0.5516 1 1>}
  n4 = StaticSwitchParameter(
   True=LinearInterpolate(A=ParticleColor().RGBA, B=DynamicParameter() {desc=Edge Color, paramNames [R,G,B,A]}.RGBA, Alpha=@n1)
   False=LinearInterpolate(A=AppendVector(A=@n2.RGB, B=@n2.RGB), B=AppendVector(A=@n3.RGB, B=@n3.RGB), Alpha=@n1)
  ) {group=Color, parameterName=Use Particle Color, sortPriority=31}
  n5 = ComponentMask(None=@n4) {a=true}
  n6 = StaticSwitchParameter(True=DepthFade(Opacity=@n5, FadeDistance=ScalarParameter() {group=Opacity, parameterName=Depth Fade}), False=@n5) {group=Opacity, parameterName=Use Depth Fade, sortPriority=31}
  n7 = ScalarParameter() {defaultValue=2, group=Edge Fade Parameters, parameterName=FADE POWER}
  n8 = Add(A=@n7, B=ScalarParameter() {group=Edge Fade Parameters, parameterName=FADE SOFTNESS})
  n9 = ComponentMask(None=TextureCoordinate()) {g=true}
  n10 = LinearInterpolate(A=@n9, B=OneMinus(None=@n9), Alpha=ScalarParameter() {group=Edge Fade Parameters, parameterName=FLIP_DIRECTION})
  LinearInterpolate(
   A=@n6
   B=Multiply(A=@n6, B=SmoothStep(Min=@n7, Max=@n8, Value=@n10) {constValue=1})
   Alpha=ScalarParameter() {group=Edge Fade Parameters, parameterName=USE_FADE}
  )
  n11 = TextureSampleParameter2D() {
   group=Overlay
   parameterName=Pattern Overlay
   texture=/BRCosmetics/Effects/Niagara/MasterMaterials/Ethereal/Textures/T_FX_Overlay_Patterns_Dirty_01
  }
  n12 = MaterialFunctionCall!(Height=@n11.R, Strength=ScalarParameter() {defaultValue=0.25, group=Overlay, parameterName=UV Distortion Strength, sortPriority=37}) {materialFunction=/BRCosmetics/Effects/Niagara/MasterMaterials/Functions/MF_FX_SimpleNormalFromMask}
  n13 = TextureCoordinate()
  n14 = Add(A=@n13, B=ScreenPosition().ViewportUV)
  n15 = ScalarParameter() {defaultValue=1, group=Overlay, parameterName=U Tile, sortPriority=33}
  n16 = ScalarParameter() {defaultValue=1, group=Overlay, parameterName=V Tile, sortPriority=34}
  n17 = Panner(
   Coordinate=Multiply(A=StaticSwitchParameter(True=@n14, False=@n13) {defaultValue=true, parameterName=ScreenPosition}, B=AppendVector(A=@n15, B=@n16))
   Speed=AppendVector(
    A=ScalarParameter() {defaultValue=0.01, group=Overlay, parameterName=U Scroll Speed, sortPriority=35}
    B=ScalarParameter() {group=Overlay, parameterName=V Scroll Speed, sortPriority=36}
   )
  )
  n18 = TextureSampleParameter2D(UVs=StaticSwitchParameter(True=Add(A=@n12.RG, B=@n17), False=@n17) {defaultValue=true, parameterName=NormalMask}) {
   group=Overlay
   parameterName=Pattern Overlay
   texture=&T_FX_Overlay_Patterns_Dirty_01
  }
  n19 = Multiply(A=@n18.RGB, B=ScalarParameter() {defaultValue=1, group=Overlay, parameterName=Overlay Strength, sortPriority=38})
  n20 = ComponentMask(None=@n4) {b=true, g=true, r=true}
  n21 = StaticSwitchParameter(
   True=LinearInterpolate(
    A=@n20
    B=Multiply(A=@n19, B=@n20)
    Alpha=ScalarParameter() {defaultValue=0.5, group=Overlay, parameterName=Overlay Opacity, sortPriority=39}
   )
   False=@n20
  ) {defaultValue=true, group=Overlay, parameterName=Use Texture Overlay, sortPriority=31}
  MaterialFunctionCall(Emissive=Multiply(A=@n21, B=ScalarParameter() {group=Color, parameterName=Emissive Strength, sortPriority=34})) {materialFunction=&MF_FX_EmissiveAdjustments}
  Reroute(None=@n21)
 ]
}
